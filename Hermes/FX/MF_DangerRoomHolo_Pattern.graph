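# Material function that generates an animated grid-cell pattern as a ShapeAlpha mask, apparently for the Danger Room hologram effect.
# A static switch tiles either mesh UVs or object-stabilized screen-space UVs.
# Each cell's rounded-up coordinate is hashed with Time as seed, giving flickering random values remapped between min and max parameters.
# One-axis SphereMasks on the fractional UVs form cell-edge lines, combined with the flicker through Max.
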MaterialFunction /Hermes/FX/MF_DangerRoomHolo_Pattern
graph {nodes [
  n1 = ScalarParameter() {defaultValue=48, parameterName=CubeUVScale_X}
  n2 = ScalarParameter() {defaultValue=24, parameterName=CubeUVScale_Y}
  n3 = StaticSwitchParameter(
   True=Multiply(A=TextureCoordinate(), B=AppendVector(A=@n1, B=@n2))
   False=MaterialFunctionCall!(Position=ObjectPositionWS(), UVRadialScale=ScalarParameter() {defaultValue=0.01, parameterName=ScreenUVScale}) {materialFunction=/Game/Weapons/WeaponSkins/MaterialFunction/ObjectStabilizedScreenSpaceUVs}.StabilizedUVs
  ) {parameterName=CubePatternUseRegularUV}
  n4 = ComponentMask(None=MaterialFunctionCall(Seed=Time(), Vec2=Ceil(None=@n3)) {materialFunction=&MF_Hash}.Vec2) {r=true}
  n5 = LinearInterpolate(
   A=ScalarParameter() {defaultValue=0.25, parameterName=Cell_Min_Value}
   B=ScalarParameter() {defaultValue=1, parameterName=Cell_Max_Value}
   Alpha=NamedRerouteDeclaration(None=Saturate(None=@n4)) {name=AnimatedCells, nodeColor=<rgba 0 0.8594 1 1>}
  )
  n6 = MaterialFunctionCall(Float2=Frac(None=@n3)) {materialFunction=&BreakOutFloat2Components}
  n7 = Constant()
  n8 = ScalarParameter() {defaultValue=0.95, parameterName=CellWidth}
  n9 = ScalarParameter() {defaultValue=0.9, parameterName=CellHardness}
  n10 = OneMinus(None=Multiply(A=SphereMask(A=@n6.R, B=@n7, Radius=@n8, Hardness=@n9), B=SphereMask(A=@n6.G, B=@n7, Radius=@n8, Hardness=@n9)))
  FunctionOutput(None=NamedRerouteDeclaration(None=Max(A=@n5, B=@n10)) {name=CellShape, nodeColor=<rgba 1 0.4453 0 1>}) {outputName=ShapeAlpha}
 ]}
