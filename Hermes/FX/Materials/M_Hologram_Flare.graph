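# Additive unlit Niagara sprite material for a hologram flare effect, built from a flare texture tinted by particle color.
# DeriveHDRfromLDR remaps the color, then per-channel powers skew red and blue oppositely through an Abbe parameter for dispersion-like color shifts.
# Hue and saturation offsets in HSV, a brightness multiplier and MF_FX_EmissiveAdjustments finish the color.
# Particle alpha drives depth-faded opacity, with optional camera-distance fade, off by default.
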Material /Hermes/FX/Materials/M_Hologram_Flare
graph {
 properties {
  bAutomaticallySetUsageInEditor=false
  bUsedWithNiagaraSprites=true
  blendMode=BLEND_Additive
  shadingModel=MSM_Unlit
  usedShadingModels=Unlit
 }
 nodes [
  n1 = Subtract(A=Distance(A=CameraPositionWS(), B=ObjectPositionWS())) {constB=100}
  n2 = ScalarParameter() {defaultValue=4000, parameterName=FadeDistance}
  n3 = Saturate(None=OneMinus(None=Divide(A=@n1, B=@n2)))
  n4 = ParticleColor()
  DepthFade(
   Opacity=LinearInterpolate(A=@n4.A, B=Multiply(A=@n4.A, B=@n3), Alpha=ScalarParameter() {parameterName=UseFadeByDistance?})
   FadeDistance=ScalarParameter() {defaultValue=50, parameterName=DepthFade}
  )
  n5 = Multiply(
   A=TextureSampleParameter2D() {group=BaseFlare, parameterName=BaseFlare, texture=/Game/Effects/Fort_Effects/Textures/Flares/T_Flare_Spikeball}.RGB
   B=@n4.RGB
  )
  n6 = MaterialFunctionCall(Float3=MaterialFunctionCall(LDR Input=Clamp(None=@n5), HDR Tint=@n5) {materialFunction=&DeriveHDRfromLDR}.RemappedRange) {materialFunction=&BreakOutFloat3Components}
  n7 = ScalarParameter() {parameterName=ChannelGammaAbbe, sliderMax=3}
  n8 = ScalarParameter() {defaultValue=1, parameterName=ChannelGamma, sliderMax=3}
  n9 = Power(Base=@n6.R, Exp=Add(A=@n8, B=@n7)) {constExponent=1.5}
  n10 = Power(Base=@n6.G, Exp=@n8) {constExponent=1}
  n11 = Power(Base=@n6.B, Exp=Subtract(A=@n8, B=@n7)) {constExponent=1.2}
  n12 = MaterialFunctionCall(Float3=RgbToHsv!(None=MaterialFunctionCall(X=@n9, Y=@n10, Z=@n11) {materialFunction=&MakeFloat3}.Result)) {materialFunction=&BreakOutFloat3Components}
  n13 = MaterialFunctionCall(
   X=Add(A=ScalarParameter() {parameterName=Hue Offset, sliderMax=1, sliderMin=-1}, B=@n12.R)
   Y=Add(A=ScalarParameter() {parameterName=Sat Offset, sliderMax=1, sliderMin=-1}, B=@n12.G)
   Z=@n12.B
  ) {materialFunction=&MakeFloat3}
  MaterialFunctionCall(Emissive=Multiply(A=HsvToRgb!(None=@n13.Result), B=ScalarParameter() {defaultValue=1, parameterName=FinalBrightness, sliderMax=1})) {materialFunction=&MF_FX_EmissiveAdjustments}
 ]
}
