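# Material function for the SpiralClouds tech art set, named as the source of the clouds' scroll speed.
# Its single output is wired to a Constant left at its default of zero, so it supplies no motion as saved.
# A Time node scaled by a small negative multiplier sits in the graph unconnected to the output, so that time-based scroll has no effect.
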MaterialFunction /Hermes/Techart/SpiralClouds/MF_SpiralClouds_ScrollSpeed
graph {nodes [
  FunctionOutput(None=Constant())
  Multiply(A=Time()) {constB=-0.013}
 ]}
