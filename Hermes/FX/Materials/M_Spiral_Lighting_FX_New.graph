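# Translucent, unlit, two-sided lightning material for Hermes spiral cloud effects.
# Vertex-offset math builds camera-facing strips from rotated spiral tangents, width patterns and wobble, alongside a newer spline-thickness branch reading position, tangent and thickness texture arrays.
# Shading branches combine distorted panning textures, HDR color, spiral lightning functions, timed flicker, and distance and time-of-day fades.
# Static switches select Niagara particle inputs or material parameters, and optional stepped panning.
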Material /Hermes/FX/Materials/M_Spiral_Lighting_FX_New
graph {
 properties {bAutomaticallySetUsageInEditor=false, blendMode=BLEND_Translucent, shadingModel=MSM_Unlit, twoSided=true, usedShadingModels=Unlit}
 nodes [
  n1 = Add(A=ComponentMask(None=TextureCoordinate()) {g=true}) {constB=-0.5}
  n2 = OneMinus(None=Multiply(A=Abs(None=@n1)) {constB=2})
  n3 = Constant3Vector() {constant=<rgba 0 0 1 0>}
  n4 = MaterialFunctionCall() {materialFunction=/Hermes/Techart/SpiralClouds/MF_SpiralClouds_ScrollSpeed}
  n5 = Constant()
  n6 = Transform(None=Constant3Vector() {constant=<rgba -1 0 0 0>})
  n7 = NamedRerouteDeclaration(None=Add(A=RotateAboutAxis(NormalizedRotationAxis=@n3, RotationAngle=@n4.Result, PivotPoint=@n5, Position=@n6), B=@n6)) {name=Rotated World Tangent, nodeColor=<rgba 1 0 0.2812 1>}
  n8 = WorldPosition() {worldPositionShaderOffset=WPT_CameraRelative}
  n9 = RotateAboutAxis(NormalizedRotationAxis=@n3, RotationAngle=@n4.Result, PivotPoint=ActorPositionWS() {originType=CameraRelative}, Position=@n8.XYZ)
  n10 = Normalize(VectorInput=Add(A=NamedRerouteDeclaration(None=@n9) {name=Origin Rotation, nodeColor=<rgba 1 0 0.3281 1>}, B=@n8.XYZ))
  n11 = Normalize(VectorInput=CrossProduct(A=@n7, B=@n10))
  n12 = NamedRerouteDeclaration(None=CrossProduct(A=@n7, B=@n11)) {name=Flat Normal, nodeColor=<rgba 1 0 0.6094 1>}
  n13 = Add(A=ComponentMask(None=TextureCoordinate()) {g=true}) {constB=-0.5}
  n14 = Add(
   A=Multiply(A=Multiply(A=@n2, B=@n12)) {constB=0.2}
   B=NamedRerouteDeclaration(None=Multiply(A=@n13, B=@n11)) {name=Cross World, nodeColor=<rgba 1 0.0703 0 1>}
  )
  n15 = NamedRerouteDeclaration(None=MaterialFunctionCall!(TexCoord=TextureCoordinate()) {materialFunction=/Hermes/Techart/SpiralClouds/MF_SpiralClouds_ThicknessPattern}.Result) {name=Thickness Pattern, nodeColor=<rgba 0 0.9062 1 1>}
  n16 = LinearInterpolate(
   A=ScalarParameter() {parameterName=Spline Width 1}
   B=ScalarParameter() {defaultValue=6000, parameterName=Spline Width 2}
   Alpha=Max(A=@n15) {constB=0.15}
  )
  n17 = Multiply(
   A=@n16
   B=MaterialFunctionCall() {materialFunction=&ObjectScale}.Scale XYZ
  )
  n18 = NamedRerouteUsage()
  n19 = Add(A=Multiply(A=Multiply(A=@n14, B=@n17) {constB=3000}, B=@n18), B=NamedRerouteUsage())
  n20 = Add(A=@n19, B=MaterialFunctionCall!(WPO=@n19) {materialFunction=/Hermes/Techart/SpiralClouds/MF_SpiralClouds_WobblePattern}.Result)
  n21 = Normalize(VectorInput=Add(A=@n20, B=WorldPosition() {worldPositionShaderOffset=WPT_CameraRelative}.XYZ))
  n22 = LinearInterpolate(
   A=ScalarParameter() {defaultValue=200, parameterName=Spline Camera Offset 1}
   B=ScalarParameter() {defaultValue=8000, parameterName=Spline Camera Offset 2}
   Alpha=Min(A=@n18, B=@n15)
  )
  n23 = Add(A=NamedRerouteUsage(), B=NamedRerouteUsage())
  Add(A=Subtract(A=@n20, B=Multiply(A=@n21, B=@n22) {constB=-2500}), B=@n23)
  Constant() {r=1}
  Constant() {r=1}
  Constant() {r=1}
  Constant()
  DynamicParameter() {paramNames [MaxGrowthLength, NewLightningTech_InvertDir, WPORand, NewLightningTech_Index]}
  DynamicParameter() {paramNames [MaxGrowthLength, NewLightningTech_InvertDir, WPORand, NewLightningTech_Index]}
  DynamicParameter() {
   parameterIndex=1
   defaultValue=<rgba 1 2 -1 0>
   paramNames [NewLightningTech_WPOStr, MoveAlignLightningDirSphereMaskPower, TextureDissolve, TextureWPONoise]
  }
  DynamicParameter() {
   parameterIndex=2
   defaultValue=<rgba 5 0 500 30>
   paramNames [ScaleSplineWidth, BaseWitdthOffset, TextureRandom, MoveAlignLightningDirSphereMaskSize]
  }
  n24 = NamedRerouteUsage()
  n25 = MaterialFunctionCall!(Lightning Alpha) {materialFunction=/Hermes/Techart/SpiralClouds/MF_SpiralClouds_Lightning}
  n26 = Multiply(A=ComponentMask(None=@n25.Result) {b=true}, B=NamedRerouteUsage())
  LinearInterpolate(
   A=ScalarParameter() {defaultValue=1, parameterName=OpacityClamp}
   Alpha=Multiply(A=Multiply(A=@n24, B=@n26), B=ScalarParameter() {defaultValue=10, parameterName=FinalOpacity})
  ) {constA=0.25}
  MaterialFunctionCall() {materialFunction=&MF_SpiralClouds_Lightning}
  n27 = NamedRerouteUsage()
  n28 = Multiply(A=NamedRerouteUsage(), B=NamedRerouteUsage())
  Multiply(A=VectorParameter() {parameterName=TempColor, defaultValue=<rgba 1 0 1 1>}.RGB, B=Multiply(A=@n25.Result, B=Multiply(A=@n27, B=@n28)))
  Multiply(A=@n23, B=MaterialFunctionCall() {materialFunction=&MF_SpiralClouds_Lightning}.Result)
  n29 = ParticleColor()
  n30 = VertexColor()
  Multiply(A=@n26, B=StaticSwitchParameter(True=Multiply(A=@n29.A, B=@n30.R), False=VertexColor().B) {parameterName=UseInNiagara?})
  n31 = LinearInterpolate(
   A=ComponentMask(None=VectorParameter() {parameterName=PanSpeed, defaultValue=<rgba 0 0.25 0 1>}.RGB) {g=true, r=true}
   B=ComponentMask(None=VectorParameter() {parameterName=PanSpeedMax, defaultValue=<rgba 0 0.5 0 1>}.RGB) {g=true, r=true}
   Alpha=VertexColor().G
  )
  Multiply(A=NamedRerouteUsage(), B=@n31)
  Multiply()
  NamedRerouteDeclaration(None=Multiply(A=@n16, B=@n18)) {name=Tangent Offset, nodeColor=<rgba 0 1 0.4375 1>}
  n32 = MaterialFunctionCall(Float2=TextureCoordinate()) {materialFunction=&BreakOutFloat2Components}
  n33 = TextureSampleParameter2DArray!(
   UVs=MaterialFunctionCall(R=OneMinus(None=@n32.G), G=Constant() {r=0.5}) {materialFunction=&AppendMany}.RGB
   Apply View MipBias
  ) {
   group=NewLightningTech
   parameterName=LightningTexture_ThicknessArray
   samplerType=SAMPLERTYPE_LinearColor
   texture=/Game/Effects/Fort_Effects/GameplayPulginsAssets/ToonLightning/GP29/Texture/T_GP29Lightning_ThicknessArray
  }
  NamedRerouteDeclaration(None=@n33.R) {name=NewLightningTech_ThickTech, nodeColor=<rgba 1 0.9141 0 1>}
  n34 = TextureCoordinate()
  NamedRerouteDeclaration(None=StaticSwitchParameter(True=MaterialFunctionCall(XY=@n34) {materialFunction=&Swizzle}.YX, False=@n34) {parameterName=SwizzleUV}) {name=TUV, nodeColor=<rgba 0.8281 1 0 1>}
  n35 = ComponentMask(None=TextureCoordinate()) {r=true}
  n36 = Multiply(A=@n35, B=OneMinus(None=@n35))
  NamedRerouteDeclaration(None=Power(Base=Multiply(A=@n36) {constB=2}) {constExponent=0.6}) {name=Length Mask, nodeColor=<rgba 0 1 0.1797 1>}
  n37 = OneMinus(None=ComponentMask(None=TextureCoordinate()) {g=true})
  n38 = VectorParameter() {parameterName=Lighting_Color, defaultValue=<rgba 0.7672 0 1 1>}
  n39 = VectorParameter() {parameterName=SecondaryGlowColor, defaultValue=<rgba 60 55.9389 21.5625 1>}
  n40 = Multiply(A=MaterialFunctionCall() {materialFunction=&LinearGradient}.VGradient)
  n41 = StaticSwitchParameter(
   True=Multiply(A=Clamp(None=@n37), B=ParticleColor().RGB)
   False=LinearInterpolate(
    A=LinearInterpolate(A=@n38.RGB, B=@n39.RGB, Alpha=@n40) {constB=0}
    B=VectorParameter() {parameterName=Lighting_Color, defaultValue=<rgba 0.7672 0 1 1>}.RGB
    Alpha=NamedRerouteUsage()
   )
  ) {parameterName=UseInNiagara?}
  n42 = DynamicParameter() {defaultValue=<rgba 1 25 1 1>, paramNames [Tile_X, HDR_Power, DistortSpeed_X, DistortSpeed_Y]}
  n43 = ScalarParameter() {defaultValue=5, parameterName=HDR_Power}
  NamedRerouteDeclaration(None=MaterialFunctionCall(LDR Input=Clamp(None=@n41), HDR Intensity Multiplier=StaticSwitchParameter(True=@n42.HDR_Power, False=@n43) {parameterName=UseInNiagara?}) {materialFunction=&DeriveHDRfromLDR}.RemappedRange) {name=HDR Color, nodeColor=<rgba 0 1 0.9531 1>}
  n44 = DynamicParameter() {defaultValue=<rgba 1 25 1 1>, paramNames [Tile_X, HDR_Power, DistortSpeed_X, DistortSpeed_Y]}
  n45 = VectorParameter() {parameterName=DistortionSpeed, defaultValue=<rgba 0 0.25 0 1>}
  NamedRerouteDeclaration(None=StaticSwitchParameter(True=AppendVector(A=@n44.DistortSpeed_X, B=@n44.DistortSpeed_X), False=ComponentMask(None=@n45.RGB) {g=true, r=true}) {parameterName=UseInNiagara?}) {name=DistortSpeed, nodeColor=<rgba 0 1 0.8828 1>}
  n46 = DynamicParameter() {
   parameterIndex=1
   defaultValue=<rgba 1 2 -1 0>
   paramNames [NewLightningTech_WPOStr, MoveAlignLightningDirSphereMaskPower, TextureDissolve, TextureWPONoise]
  }
  n47 = Multiply(A=@n46.MoveAlignLightningDirSphereMaskPower, B=LinearInterpolate(Alpha=VertexColor().G) {constA=0.5})
  n48 = ScalarParameter() {defaultValue=1, parameterName=TextureTile_X}
  n49 = LinearInterpolate(
   A=ScalarParameter() {defaultValue=1, parameterName=TextureTile_Y}
   B=ScalarParameter() {defaultValue=1, parameterName=TextureTile_Y_Max}
   Alpha=VertexColor().G
  ) {constA=0.5}
  NamedRerouteDeclaration(None=StaticSwitchParameter(True=AppendVector(A=@n46.NewLightningTech_WPOStr, B=@n47), False=AppendVector(A=@n48, B=@n49)) {parameterName=UseInNiagara?}) {name=TextureTile, nodeColor=<rgba 1 0 0.3984 1>}
  n50 = ScalarParameter() {defaultValue=0.2, parameterName=DistortionAmount}
  n51 = ScalarParameter() {defaultValue=0.2, parameterName=DistortionAmountMax}
  n52 = VertexColor()
  n53 = TextureCoordinate() {uTiling=0.46, vTiling=0.2476}
  n54 = Time()
  n55 = ScalarParameter() {defaultValue=1, parameterName=UVDistortSpeed}
  n56 = Divide(A=Ceil(None=Multiply(A=@n54, B=@n55) {constB=16})) {constB=4}
  n57 = NamedRerouteUsage()
  n58 = TextureCoordinate() {uTiling=0.46, vTiling=0.2476}
  n59 = Time()
  n60 = NamedRerouteUsage()
  n61 = TextureSampleParameter2D(UVs=StaticSwitchParameter(True=Panner(Coordinate=@n53, Time=@n56, Speed=@n57) {speedY=0.4}, False=Panner(Coordinate=@n58, Time=@n59, Speed=@n60) {speedY=0.4}) {parameterName=UseStepping?}) {
   parameterName=DistortionTexture
   samplerType=SAMPLERTYPE_LinearColor
   texture=/Game/Global/Textures/Default/Noise/TilingNoise06_N_uncompressed
  }
  n62 = Constant()
  n63 = MaterialFunctionCall(Float2=Add(A=ComponentMask(None=@n61.RGB) {g=true, r=true}, B=@n62) {constB=0}) {materialFunction=&BreakOutFloat2Components}
  n64 = Clamp(None=@n63.R)
  n65 = ScalarParameter() {defaultValue=1, parameterName=Clamp Texture V}
  n66 = Multiply(
   A=StaticSwitchParameter(
    True=DynamicParameter() {parameterIndex=2, paramNames [MainScaler, PannerMask_Speed, Distortion, Param4]}.Distortion
    False=LinearInterpolate(A=@n50, B=@n51, Alpha=@n52.G)
   ) {parameterName=UseInNiagara?}
   B=AppendVector(A=LinearInterpolate(A=@n63.R, B=@n64, Alpha=@n65), B=OneMinus(None=@n63.G))
  )
  n67 = TextureCoordinate()
  n68 = NamedRerouteUsage()
  n69 = ScalarParameter() {defaultValue=16, parameterName=X_Offset}
  n70 = Constant()
  n71 = Panner(Coordinate=Add(A=Multiply(A=@n67, B=@n68), B=AppendVector(A=@n69, B=@n70)), Speed=@n31) {speedY=0.5}
  NamedRerouteDeclaration(None=TextureSampleParameter2D(UVs=Add(A=@n66, B=@n71)) {parameterName=TilingTexture, texture=/Game/Environments/Hexmap/HexClouds/Lightning/T_Hexmap_Lightning_02_E}.B) {name=LightingTextureOpactiy, nodeColor=<rgba 0 0.1562 1 1>}
  n72 = Multiply(A=NamedRerouteUsage()) {constB=-1}
  NamedRerouteDeclaration(None=VertexInterpolator(VS=DotProduct(A=@n10, B=@n72)).PS) {name=Fresnel, nodeColor=<rgba 0 0.0625 1 1>}
  n73 = ScalarParameter() {defaultValue=100000, parameterName="Fade_Start_Distance (S)"}
  n74 = ScalarParameter() {defaultValue=50000, parameterName="Fade_End_Distance (S)"}
  NamedRerouteDeclaration(None=LinearInterpolate(Alpha=MaterialFunctionCall(Fade_Start_Distance=@n73, Fade_End_Distance=@n74) {materialFunction=&MF_Distance_SyncedFade_FX}.Result) {constA=1, constB=0.35}) {name=Distance_Intensity_Fade, nodeColor=<rgba 0 0.4844 1 1>}
  NamedRerouteDeclaration(None=ComponentMask(None=NamedRerouteUsage()) {g=true}) {name=Size Fade, nodeColor=<rgba 0 1 0.8125 1>}
  n75 = NamedRerouteUsage()
  NamedRerouteDeclaration(None=Multiply(A=NamedRerouteUsage(), B=Multiply(A=@n75))) {name=DistanceFading_Complete, nodeColor=<rgba 0.8513 0.4325 1 1>}
  n76 = ScalarParameter() {defaultValue=2500, parameterName=FarFadeout_Distance}
  n77 = ScalarParameter() {defaultValue=-2500, parameterName=FarFadeout_Distance_Offset}
  NamedRerouteDeclaration(None=OneMinus(None=MaterialFunctionCall!(Blend Range=@n76, Start Offset=@n77) {materialFunction=/Engine/Functions/Engine_MaterialFunctions03/Particles/Distance_Blend}.Result)) {name=DistanceFade, nodeColor=<rgba 0.9155 0.651 1 1>}
  n78 = MaterialFunctionCall(
   Blend Range=ScalarParameter() {defaultValue=2500, parameterName=CloseFadeout_Distance}
   Start Offset=ScalarParameter() {defaultValue=-2500, parameterName=CloseFadeout_Distance_Offset}
  ) {materialFunction=&Distance_Blend}
  NamedRerouteDeclaration(None=@n78.Result) {name=Close_DistanceFade, nodeColor=<rgba 0.8285 0.2917 1 1>}
  n79 = Divide(A=CameraPositionWS()) {constB=15000}
  NamedRerouteDeclaration(None=LinearInterpolate(
    A=ScalarParameter() {defaultValue=0.25, parameterName=GroundIntensity}
    B=ScalarParameter() {defaultValue=1, parameterName=InAirIntensity}
    Alpha=Clamp(None=@n79)
   )) {name=CameraPos_Intensity, nodeColor=<rgba 1 0.0703 0 1>}
  NamedRerouteDeclaration(None=MaterialFunctionCall(
    Value_Morning=ScalarParameter() {defaultValue=1, parameterName="Value_Morning (S)"}
    Value_Day=ScalarParameter() {defaultValue=1, parameterName="Value_Day (S)"}
    Value_Evening=ScalarParameter() {defaultValue=0.5, parameterName="Value_Evening (S)"}
    Value_Night=ScalarParameter() {defaultValue=0.75, parameterName="Value_Night (S)"}
   ) {materialFunction=&Scalar_Time_of_Day}.Result) {name=TimeOfDay_Intensity, nodeColor=<rgba 0.1484 1 0 1>}
  n80 = OneMinus(None=MaterialFunctionCall(Float2=TextureCoordinate()) {materialFunction=&BreakOutFloat2Components}.G)
  n81 = Multiply(A=ComponentMask(None=VertexColor().R) {r=true}) {constB=10}
  n82 = Floor(None=Add(A=Floor(None=@n81), B=ScalarParameter() {group=NewLightningTech, parameterName=NewLightningTech_DebugIndex}))
  n83 = TextureSampleParameter2DArray(UVs=MaterialFunctionCall(R=@n80, G=Constant() {r=0.5}, B=@n82) {materialFunction=&AppendMany}.RGB) {
   group=NewLightningTech
   parameterName=LightningTexture_PositionArray
   samplerType=SAMPLERTYPE_LinearColor
   texture=/Game/Effects/Fort_Effects/GameplayPulginsAssets/ToonLightning/GP29/Texture/T_GP29Lightning_PositionArray
  }
  n84 = MaterialFunctionCall(Float3=Multiply(A=Subtract(A=@n83.RGB) {constB=0.5}) {constB=2}) {materialFunction=&BreakOutFloat3Components}
  n85 = DynamicParameter() {paramNames [MaxGrowthLength, NewLightningTech_InvertDir, WPORand, NewLightningTech_Index]}
  n86 = LinearInterpolate(Alpha=Saturate(None=Floor(None=@n85.NewLightningTech_InvertDir))) {constA=1, constB=-1}
  n87 = ScalarParameter() {defaultValue=1, group=NewLightningTech, parameterName=NewLightningTech_ScaleZ}
  n88 = DynamicParameter() {
   parameterIndex=1
   defaultValue=<rgba 1 2 -1 0>
   paramNames [NewLightningTech_WPOStr, MoveAlignLightningDirSphereMaskPower, TextureDissolve, TextureWPONoise]
  }
  n89 = Multiply(
   A=MaterialFunctionCall(X=Multiply(A=@n84.R, B=@n86), Y=Multiply(A=@n84.G, B=@n86), Z=Divide(A=@n84.B, B=@n87)) {materialFunction=&MakeFloat3}.Result
   B=Multiply(A=ScalarParameter() {defaultValue=20, group=NewLightningTech, parameterName=NewLightningTech_WPOStr}, B=@n88.NewLightningTech_WPOStr)
  )
  n90 = Transform(None=@n89) {transformSourceType=TRANSFORMSOURCE_View}
  n91 = Constant() {r=1}
  n92 = TextureSampleParameter2DArray(UVs=MaterialFunctionCall(R=@n80, G=Constant() {r=0.5}, B=@n82) {materialFunction=&AppendMany}.RGB) {
   group=NewLightningTech
   parameterName=LightningTexture_TangentArray
   samplerType=SAMPLERTYPE_LinearColor
   texture=/Game/Effects/Fort_Effects/GameplayPulginsAssets/ToonLightning/GP29/Texture/T_GP29Lightning_TangentArray
  }
  n93 = Subtract(A=@n92.RGB) {constB=0.5}
  n94 = Transform(None=Multiply(
    A=MaterialFunctionCall(X=@n86, Y=@n86, Z=@n91) {materialFunction=&MakeFloat3}.Result
    B=Multiply(A=@n93) {constB=2}
   )) {transformSourceType=TRANSFORMSOURCE_Instance}
  n95 = Subtract(A=CameraPositionWS(), B=Add(A=WorldPosition().XYZ, B=@n90))
  n96 = MaterialFunctionCall() {materialFunction=&TangentBasis}
  n97 = LinearInterpolate(
   A=Normalize(VectorInput=CrossProduct(A=@n94, B=@n95))
   B=Normalize(VectorInput=CrossProduct(A=@n96.Y V, B=@n95))
   Alpha=ScalarParameter() {group=NewLightningTech, parameterName=NewLightningTech_TangentBlend}
  )
  n98 = NamedRerouteUsage()
  n99 = DynamicParameter() {
   parameterIndex=3
   defaultValue=<rgba 0 0 1 -0.7>
   paramNames [MoveAlignLightningDir, "MoveToCenter(NoPerf/Mobile)", Growth, ThicknessDissolve]
  }
  n100 = Saturate(None=Multiply(A=Subtract(A=@n98, B=@n99.ThicknessDissolve), B=ScalarParameter() {defaultValue=2, parameterName=ThicknessDissolveHardness}))
  n101 = MaterialFunctionCall(Float2=NamedRerouteUsage()) {materialFunction=&BreakOutFloat2Components}
  n102 = LinearInterpolate(A=@n101.G, B=OneMinus(None=@n101.G), Alpha=ScalarParameter() {defaultValue=1, parameterName=SplineThickness_GrowthDirectionInvert})
  n103 = DynamicParameter() {
   parameterIndex=3
   defaultValue=<rgba 0 0 1 -0.7>
   paramNames [MoveAlignLightningDir, "MoveToCenter(NoPerf/Mobile)", Growth, ThicknessDissolve]
  }
  n104 = Multiply(
   A=Saturate(None=@n103.Growth)
   B=DynamicParameter() {paramNames [MaxGrowthLength, NewLightningTech_InvertDir, WPORand, NewLightningTech_Index]}.MaxGrowthLength
  )
  n105 = ScalarParameter() {defaultValue=5, parameterName=SplineThickness_GrowthGradient}
  n106 = Add(B=Divide(B=@n105)) {constA=1}
  n107 = Multiply(A=Add(A=Subtract(A=@n102), B=Multiply(A=@n104, B=@n106)), B=@n105)
  n108 = Multiply(A=Add(A=@n102, B=@n103.Growth), B=@n105)
  n109 = ScalarParameter() {defaultValue=100, parameterName=SplineThickness_BaseWidth}
  n110 = ScalarParameter() {defaultValue=5, parameterName=ScaleSplineWidth}
  n111 = StaticSwitchParameter(True=@n109, False=Max(B=Add(A=@n110, B=@n109))) {parameterName=Glow}
  n112 = ScalarParameter() {defaultValue=600, parameterName=SplineThickness_SpikeWidth}
  n113 = ScalarParameter() {defaultValue=0.1, parameterName=SplineThickness_SpikeDataSubtract}
  n114 = Subtract(A=NamedRerouteUsage(), B=@n113)
  n115 = Power(
   Base=Divide(A=Saturate(None=@n114), B=OneMinus(None=@n113))
   Exp=ScalarParameter() {defaultValue=1, parameterName=SplineThickness_SpikeDataPower}
  )
  n116 = Multiply(
   A=StaticSwitchParameter(True=Saturate(None=@n107), False=Saturate(None=@n108)) {parameterName=UseNewGrowthMethod}
   B=Multiply(A=LinearInterpolate(A=@n111, B=@n112, Alpha=@n115), B=@n110)
  )
  n117 = VectorParameter() {
   group=WidthEdgeMask
   parameterName=WidthEdgeMask_EdgeMask_InvertControl
   channelNames {a="", b=Power, g=Multiply, r=Offset}
   defaultValue=<rgba 0 10 1 1>
  }
  n118 = VectorParameter() {
   group=WidthEdgeMask
   parameterName=WidthEdgeMask_EdgeMask_Control
   channelNames {a="", b=Power, g=Multiply, r=Offset}
   defaultValue=<rgba 0 10 1 1>
  }
  n119 = MaterialFunctionCall!(
   UorV=MaterialFunctionCall(Float2=NamedRerouteUsage()) {materialFunction=&BreakOutFloat2Components}.G
   Invert_EdgeControl=@n117
   NoInvert_EdgeControl=@n118
  ) {materialFunction=/Game/Effects/Niagara/MasterMaterials/Functions/Common/FXCommon_EdgeMask_Ex}
  n120 = Constant() {r=1}
  n121 = VectorParameter() {parameterName=WidthEdgeMask_Invert, defaultValue=<rgba 0 0.2 0 1>}
  n122 = MaterialFunctionCall(Input=OneMinus(None=@n119.Mask), Input Low=Constant(), Input High=@n120, Target Low=@n121.G, Target High=@n120) {materialFunction=&RemapValueRange}
  n123 = Constant() {r=-0.5}
  n124 = TextureCoordinate()
  n125 = Multiply(
   A=Multiply(A=Multiply(A=@n100, B=@n116), B=LinearInterpolate(A=@n119.Mask, B=@n122.Result, Alpha=@n121.R))
   B=ComponentMask(None=Add(A=@n123, B=@n124)) {r=true}
  )
  NamedRerouteDeclaration(None=Add(A=@n90, B=Multiply(A=@n97, B=@n125))) {name=WPO_SplineThickness, nodeColor=<rgba 0.4766 1 0 1>}
  n126 = TextureObjectParameter() {
   group=TextureWPONoise
   parameterName=TextureWPONoise_Texture
   sortPriority=0
   texture=/Game/Effects/Fort_Effects/Emotes/UltraLight/Texture/T_SoftNoise_BC7
  }
  n127 = VectorParameter() {
   group=TextureWPONoise
   parameterName=TextureWPONoise_UVControl
   channelNames {a=VPan, b=UPan, g=VScale, r=UScale}
   defaultValue=<rgba 0 1 0 0.5>
  }
  n128 = Multiply(A=ComponentMask(None=VertexColor().R) {r=true}) {constB=10}
  n129 = Add(A=Time() {bOverride_Period=true, period=6666}, B=@n128)
  n130 = NamedRerouteUsage()
  n131 = Multiply(
   A=Add(
    A=MaterialFunctionCall(Texture=@n126, UVControl_RGB=@n127, UVControl_A=@n127, "[Override]TimeInput"=@n129, "[Override]UVInput"=@n130) {materialFunction=&FXCommon_TextureSample}.RGB
    B=ScalarParameter() {defaultValue=-0.5, parameterName=TextureWPONoise_Offset}
   )
   B=Add(A=@n128, B=ScalarParameter() {parameterName=TextureWPONoise_BaseNoiseIntensity})
  )
  n132 = VectorParameter() {group=TextureWPONoise, parameterName=TextureWPONoise_Mask, channelNames {a=VPan, b=UPan, g=VScale, r=UScale}, defaultValue=<rgba 1 1 1 0.5>}
  n133 = MaterialFunctionCall(Float2=NamedRerouteUsage()) {materialFunction=&BreakOutFloat2Components}
  n134 = VectorParameter() {
   group=TextureWPONoise
   parameterName=TextureWPONoise_EdgeMask_InvertControl
   channelNames {a="", b=Power, g=Multiply, r=Offset}
   defaultValue=<rgba 0 10000 1 1>
  }
  n135 = VectorParameter() {
   group=TextureWPONoise
   parameterName=TextureWPONoise_EdgeMask_Control
   channelNames {a="", b=Power, g=Multiply, r=Offset}
   defaultValue=<rgba 0 10000 1 1>
  }
  n136 = Transform(None=Multiply(
    A=Multiply(A=@n131, B=@n132)
    B=MaterialFunctionCall(UorV=@n133.G, Invert_EdgeControl=@n134, NoInvert_EdgeControl=@n135) {materialFunction=&FXCommon_EdgeMask_Ex}.Mask
   )) {transformSourceType=TRANSFORMSOURCE_Local}
  n137 = Constant3Vector()
  NamedRerouteDeclaration(None=MaterialFunctionCall!("Xb1,PS4"=@n136, "PC-High&&Epic,XSX,PS5"=@n136, "PC-Perf,Android"=@n137, Switch=@n137, "PC-Low(NotPerf)&&Medium"=@n136) {materialFunction=/Game/Effects/Niagara/MasterMaterials/Functions/Common/PlatformSwitch_3D}.Result) {name=TextureWPO, nodeColor=<rgba 1 0.7031 0 1>}
  NamedRerouteDeclaration(None=@n23) {name=WPO, nodeColor=<rgba 1 0 0.7734 1>}
  NamedRerouteDeclaration(None=@n11) {name=Cross World Module, nodeColor=<rgba 1 0.0703 0 1>}
  NamedRerouteUsage()
  OneMinus(None=@n78.Result)
  n138 = Time()
  n139 = ScalarParameter() {defaultValue=25, group=Flicker, parameterName=Num Items, sliderMax=25}
  n140 = ScalarParameter() {defaultValue=0.5, group=Flicker, parameterName=Time Between Triggers}
  n141 = Multiply(A=@n140, B=@n139)
  n142 = ScalarParameter() {defaultValue=0.15, group=Flicker, parameterName=duration, sliderMax=0.25}
  n143 = Multiply(A=@n140, B=Multiply(A=@n139, B=VertexColor().R))
  Saturate(None=MaterialFunctionCall(
    Input=Modulo!(A=@n138, B=@n141) {desc=cycling timeline}
    Input Low=@n143
    Input High=Add(A=@n143, B=@n142)
    Target Low=Constant()
    Target High=Constant() {r=1}
   ) {materialFunction=&RemapValueRange}.Result)
  n144 = TextureCoordinate()
  n145 = VertexColor()
  n146 = DynamicParameter() {defaultValue=<rgba 1 25 1 1>, paramNames [Tile_X, HDR_Power, DistortSpeed_X, DistortSpeed_Y]}
  n147 = ScalarParameter() {defaultValue=1, parameterName=Dynamic V Scale}
  n148 = VertexColor()
  n149 = AppendVector(
   A=Constant()
   B=StaticSwitchParameter(True=Multiply(A=@n145.G, B=@n146.Tile_X), False=Multiply(A=@n147, B=@n148.G)) {parameterName=UseInNiagara?}
  )
  n150 = AppendVector(A=ScalarParameter() {parameterName=X Speed}, B=ScalarParameter() {defaultValue=1, parameterName=Y Speed})
  n151 = MaterialFunctionCall!(
   TimeSpeed=ScalarParameter() {defaultValue=1, group=Step Panner, parameterName=StepPan_Timespeed}
   Jump Amount Per Step=AppendVector(
    A=ScalarParameter() {group=Step Panner, parameterName=StepPan_JumpAmount_X}
    B=ScalarParameter() {defaultValue=2, group=Step Panner, parameterName=StepPan_JumpAmount_Y}
   )
   Number of Steps Per Unit of Time=AppendVector(
    A=ScalarParameter() {defaultValue=1.5, group=Step Panner, parameterName=XStepPan_NumberofStepsPerSecond}
    B=ScalarParameter() {defaultValue=1.5, group=Step Panner, parameterName=YStepPan_NumberofStepsPerSecond}
   )
   Time Offset
  ) {materialFunction=/Engine/Functions/Engine_MaterialFunctions02/SteppingPannerTime}
  n152 = ScalarParameter() {group=Step Panner, parameterName=Use_StepPanner}
  StaticSwitchParameter(
   True=Panner(Coordinate=Add(A=@n144, B=@n149), Speed=LinearInterpolate(A=@n150, B=@n151.Result, Alpha=@n152)) {bFractionalPart=true}
   False=@n62
  ) {parameterName=UseStepping?}
  Transform(None=@n89) {transformSourceType=TRANSFORMSOURCE_Instance}
 ]
}
