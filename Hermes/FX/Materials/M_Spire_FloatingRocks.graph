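# M_Spire_FloatingRocks is a Default Lit material for Niagara mesh particles, apparently for floating rock debris in the Hermes Spire effects.
# It combines particle color with a Helios Brim rock-ground diffuse texture through separate Add and Multiply nodes, allowing per-particle tinting.
# A matching normal map and a Hermes monolith rock mask are also sampled.
# The listing shows no connections from these nodes to material outputs.
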Material /Hermes/FX/Materials/M_Spire_FloatingRocks
graph {
 properties {
  bAutomaticallySetUsageInEditor=false
  bUsedWithNiagaraMeshParticles=true
  previewMesh=/Hermes/FX/Meshes/SM_FloatingRocks_Mesh_01
  usedShadingModels=Default Lit
 }
 nodes [
  n1 = ParticleColor()
  n2 = TextureSample() {texture=/Game/Environments/Helios/Terrain/Textures/T_Helios_Brim_RockGround_01_D}
  Add(A=@n1.RGB, B=@n2.RGB)
  Multiply(A=@n1.RGB, B=@n2.RGB)
  TextureSample() {samplerType=SAMPLERTYPE_Normal, texture=/Game/Environments/Helios/Terrain/Textures/T_Helios_Brim_RockGround_01_N}
  TextureSample() {samplerType=SAMPLERTYPE_Masks, texture=/Hermes/Rocks/MMSpikeyRocks/Textures/T_Hermes_Rock_MaskMaker_Monolith_bottom_MSK}
 ]
}
