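# Niagara system for a spire effect: floating rock meshes swirl and drift upward around a magenta-tinted spiral mesh.
# GPU Rocks spawn in stretched spheres with random meshes, moved by biased curl noise, point attraction and vortex forces, fading and scaling over life.
# CPU Spiral rises from below the origin; disabled Rocks_OnSpire adds depth-snapped sprites and meshes.
# All emitters use distance culling and user spawn-rate and scale multipliers.
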NiagaraSystem /Hermes/FX/NS_Spire_Floating_Rocks
dependencies {
 usedDataInterfaces [
  &NiagaraDataInterfaceCurve
  &NiagaraDataInterfaceGBuffer
  &NiagaraDataInterfaceSpriteRendererInfo
 ]
 usedDynamicInputs []
 usedModules [
  &SystemState
  &self:SystemSpawnScript.SystemScriptSource.SystemScriptGraph.NiagaraNodeAssignment_1.SetVariables_#1
 ]
}
emitters {
 Rocks {
  simTarget=GPUComputeSim
  emitterUpdateScript [
   &EmitterState(
    Enable Distance Culling=true
    MaxDistance=25000
    Min Distance Response=::SleepAndLetParticlesFinish
    Module.bApplySpawnCountScaleByDistance=true
    Module.bUseMaxDistance=true
    Module.bUseMinDistance=true
    Scalability Mode=::NewEnumerator1
    Scale Spawn Count=true
    Spawn Count Scale By Distance=DataInterface() {propertyValues {Curve {keys [<at 0 1>,<at 0.1 1>,<at 1 1>]}}}
   )
   &SpawnRate(SpawnRate=&Multiply_Float(A=4, B=$User.SpawnRatMult), Use Spawn Probability=true)
  ]
  particleSpawnScript [
   &InitializeParticle(
    Color=<rgba 0.0578 0.0595 0.0612 1>
    Lifetime Max=8
    Lifetime Min=5
    Lifetime Mode=::NewEnumerator1
    Mesh Scale Max=<0.15 0.15 0.1>
    Mesh Scale Min=<0.1 0.1 0.05>
    Mesh Scale Mode=::NewEnumerator2
    Sprite Size Mode=::NewEnumerator1
    Uniform Sprite Size Max=1
    Uniform Sprite Size Min=1
    UsePositionOffset=false
    Write Material Random=true
   ) {module=InitializeParticle001}
   &InitialMeshOrientation(
    Mesh Orientation Mode=::NewEnumerator2
    Rotation=&RandomRangeVector(Minimum=<-0.02 -0.02 0>, Maximum=<0.02 0.02 1>)
    Use Rotation Vector=true
   )
   &MeshRotationForce(Rotation=&RandomRangeVector(Minimum=<0 0 -0.5>, Maximum=<0.1 0.1 0.5>))
   &ApplyInitialForces()
   &CameraOffset(Camera Offset Amount=25)
   &ShapeLocation(Handle Radius=100, Large Radius=250, MaskSpawnGroup=false, Override Seed=false, Shape Primitive=::NewEnumerator3, UseOwnerScale=true) {enabled=false}
   &ShapeLocation(MaskSpawnGroup=false, Non Uniform Scale=<1 1 6>, Override Seed=false, Sphere Radius=500, UseOwnerScale=true) {module=ShapeLocation002}
   /Niagara/Modules/Update/Utility/SampleGBuffer()
   &ShapeLocation(MaskSpawnGroup=false, Override Seed=false, Sphere Radius=250, UseOwnerScale=true) {module=ShapeLocation003}
   /SetVariables(Particles.InitialPos:NiagaraPosition=$Particles.Position) {isSetParameters=true, module=SetVariables_#2}
   &KillParticlesInVolume(Invert Volume=true, Sphere Radius=5000)
   &AddVelocity(Velocity=<0 0 100>)
   /SetVariables(Particles.MeshIndex=&RandomRangeInt(Minimum=0, Maximum=3)) {isSetParameters=true, module=SetVariables_#3}
   &ShapeLocation(
    MaskSpawnGroup=false
    Non Uniform Scale=<0.2 0.2 1>
    Offset Mode=::NewEnumerator0
    Override Seed=false
    Sphere Radius=250
    UseOwnerScale=true
   ) {module=ShapeLocation001}
  ]
  particleUpdateScript [
   &ParticleState()
   &UpdateMeshOrientation(Rotation Rate=&RandomRangeFloat(Minimum=-0.1, Maximum=-0.25), Rotation Vector=<0 0 1>)
   &CurlNoiseForce(
    Bias Noise=true
    Bias Noise Field:Vector3f=<0 0 1>
    Noise Frequency=&RandomRangeFloat(Minimum=0.1, Maximum=0.2)
    Noise Strength=&RandomRangeFloat(Minimum=500, Maximum=25)
    Pan Noise=true
    Pan Noise Field=<1 0 1>
   )
   &PointAttractionForce(
    Attraction Radius=25000
    Attractor Position Offset:Vector3f=<0 0 200>
    Kill Within Radius=false
    Module.Use Falloff=true
    UseAttractorOffset=true
   )
   &VortexForce(UseOriginOffset=false, Vortex Force Amount=-2500)
   &SolveForcesAndVelocity(
    Acceleration Limit=25
    Clamp Velocity=true
    Limit Acceleration=true
    Speed Limit:NiagaraFloat=&Multiply_Float(
     A=150
     B=&FloatFromCurve(FloatCurve=DataInterface() {propertyValues {Curve {keys [
          <at 0 0.5>
          {arriveTangent=1.255, interpMode=RCIM_Cubic, leaveTangent=1.255, tangentMode=RCTM_User, time=0.5, value=1}
          {arriveTangent=2.5018, interpMode=RCIM_Cubic, leaveTangent=2.5018, tangentMode=RCTM_User, time=1, value=3}
         ]}}})
    )
   )
   &CalculateAccurateVelocity()
   &ScaleColor(
    Scale Alpha=&FloatFromCurve(FloatCurve=DataInterface() {propertyValues {Curve {
        defaultValue=1
        keys [
         {interpMode=RCIM_Cubic, tangentMode=RCTM_User, time=0, value=0}
         {interpMode=RCIM_Cubic, tangentMode=RCTM_User, time=0.5, value=1}
         {tangentMode=RCTM_User, time=1, value=0}
        ]
       }}})
    ScaleA=true
    ScaleRGB=true
   )
   &ScaleMeshSize(Scale Factor=&Multiply_VectorByFloat(
     Vector=&Multiply_VectorByFloat(Vector=<25 25 25>, Float=$User.ScaleMult)
     Float=&FloatFromCurve(
      FloatCurve=DataInterface() {propertyValues {Curve {
         defaultValue=1
         keys [
          {interpMode=RCIM_Cubic, tangentMode=RCTM_User, time=0, value=0}
          {interpMode=RCIM_Cubic, tangentMode=RCTM_User, time=0.5, value=1}
          {tangentMode=RCTM_User, time=1, value=0}
         ]
        }}}
      Scale Curve=/Niagara/DynamicInputs/Scalability/ScalabilityDistanceBasedFloat(
       Curve:NiagaraDataInterfaceCurve=DataInterface() {propertyValues {Curve {keys [<at 0.1 1>,<at 0.2 1>,<at 0.8 1>,<at 1 0>]}}}
       CurveIndex:NiagaraFloat=$Emitter.Scalability.DistanceFraction
       Bias:NiagaraFloat=0
       Scale:NiagaraFloat=0.5
      )
     )
    ))
   &CameraOffset(Camera Offset Amount=250) {module=CameraOffset001}
  ]
  rendererClasses [&NiagaraMeshRendererProperties]
  renderers [{rendererClass=&NiagaraMeshRendererProperties, rendererIndex=0}]
 }
 Rocks_OnSpire {
  bEnabled=false
  simTarget=GPUComputeSim
  emitterUpdateScript [
   &EmitterState(
    Enable Distance Culling=true
    MaxDistance=25000
    Min Distance Response=::SleepAndLetParticlesFinish
    Module.bApplySpawnCountScaleByDistance=true
    Module.bUseMaxDistance=true
    Module.bUseMinDistance=true
    Scalability Mode=::NewEnumerator1
    Scale Spawn Count=true
    Spawn Count Scale By Distance=DataInterface() {propertyValues {Curve {keys [<at 0 0>,<at 0.1 1>,<at 1 1>]}}}
   )
   &SpawnRate(SpawnRate=&Multiply_Float(A=10, B=$User.SpawnRatMult), Use Spawn Probability=true)
  ]
  particleSpawnScript [
   &InitializeParticle(
    Color=<rgba 0.0625 0.0625 0.0625 1>
    Lifetime Max=8
    Lifetime Min=6
    Lifetime Mode=::NewEnumerator1
    Mesh Scale Max=<2 2 3>
    Mesh Scale Min=<1 1 2>
    Mesh Scale Mode=::NewEnumerator2
    Sprite Rotation Mode=::NewEnumerator2
    Sprite Size Mode=::NewEnumerator1
    Uniform Sprite Size Max=1500
    Uniform Sprite Size Min=250
    UsePositionOffset=false
    Write Material Random=true
   ) {module=InitializeParticle001}
   &InitialMeshOrientation(
    Mesh Orientation Mode=::NewEnumerator2
    Rotation=&RandomRangeVector(Minimum=<-0.02 -0.02 0>, Maximum=<0.02 0.02 1>)
    Use Rotation Vector=true
   )
   &MeshRotationForce(Rotation=&RandomRangeVector(Minimum=<0 0 -0.5>, Maximum=<0.1 0.1 0.5>))
   &ApplyInitialForces()
   &CameraOffset(Camera Offset Amount=25)
   &ShapeLocation(Handle Radius=100, Large Radius=250, MaskSpawnGroup=false, Override Seed=false, Shape Primitive=::NewEnumerator3, UseOwnerScale=true) {enabled=false}
   &ShapeLocation(MaskSpawnGroup=false, Non Uniform Scale=<1 1 8>, Override Seed=false, Sphere Radius=500, UseOwnerScale=true) {module=ShapeLocation002}
   &SampleGBuffer(Kill Beyond Depth=false, Snap To Depth:ENiagaraGBufferSnapToDepth=::NewEnumerator1)
   /SetVariables(Particles.InitialPos:NiagaraPosition=$Particles.Position) {isSetParameters=true, module=SetVariables_#2}
   &KillParticlesInVolume(Invert Volume=true, Sphere Radius=5000)
   &AddVelocity(Velocity=<0 0 20>)
   /SetVariables(Particles.MeshIndex=&RandomRangeInt(Minimum=0, Maximum=3)) {isSetParameters=true, module=SetVariables_#3}
   &SubUVAnimation(
    Evaluation Type=::NewEnumerator0
    "Sprite Renderer "=DataInterface() {propertyValues {SpriteRenderer=&self:Rocks_OnSpire_0.NiagaraSpriteRendererProperties_0}}
    SubUV Animation Mode:ENiagara_SubUVLookupModeV2=::NewEnumerator2
    UseEndFrame=false
    UseStartFrame=false
   )
  ]
  particleUpdateScript [
   &ParticleState()
   &UpdateMeshOrientation(Rotation Rate=&RandomRangeFloat(Minimum=-0.1, Maximum=0.02), Rotation Vector=<0 0 1>)
   &CurlNoiseForce(
    Bias Noise=true
    Bias Noise Field:Vector3f=<0 0 1>
    Noise Frequency=&RandomRangeFloat(Minimum=0.1, Maximum=0.2)
    Noise Strength=&RandomRangeFloat(Minimum=0, Maximum=25)
    Pan Noise=true
    Pan Noise Field=<1 0 1>
   )
   &SolveForcesAndVelocity(
    Acceleration Limit=25
    Clamp Velocity=true
    Limit Acceleration=true
    Speed Limit:NiagaraFloat=&Multiply_Float(
     A=2500
     B=&FloatFromCurve(FloatCurve=DataInterface() {propertyValues {Curve {keys [
          <at 0 0.5>
          {arriveTangent=1.255, interpMode=RCIM_Cubic, leaveTangent=1.255, tangentMode=RCTM_User, time=0.5, value=1}
          {arriveTangent=2.5018, interpMode=RCIM_Cubic, leaveTangent=2.5018, tangentMode=RCTM_User, time=1, value=3}
         ]}}})
    )
   )
   &CalculateAccurateVelocity()
   &ScaleColor(
    Scale Alpha=&FloatFromCurve(
     FloatCurve=DataInterface() {propertyValues {Curve {
        defaultValue=1
        keys [
         {interpMode=RCIM_Cubic, tangentMode=RCTM_User, time=0, value=0}
         {interpMode=RCIM_Cubic, tangentMode=RCTM_User, time=0.5, value=1}
         {tangentMode=RCTM_User, time=1, value=0}
        ]
       }}}
     Scale Curve=0.1
    )
    ScaleA=true
    ScaleRGB=true
   )
   &ScaleMeshSize(Scale Factor=&Multiply_VectorByFloat(
     Vector=&Multiply_VectorByFloat(Vector=<1 1 1>, Float=$User.ScaleMult)
     Float=&FloatFromCurve(
      FloatCurve=DataInterface() {propertyValues {Curve {
         defaultValue=1
         keys [
          {interpMode=RCIM_Cubic, tangentMode=RCTM_User, time=0, value=0}
          {interpMode=RCIM_Cubic, tangentMode=RCTM_User, time=0.5, value=1}
          {tangentMode=RCTM_User, time=1, value=0}
         ]
        }}}
      Scale Curve=&ScalabilityDistanceBasedFloat(
       Curve:NiagaraDataInterfaceCurve=DataInterface() {propertyValues {Curve {keys [<at 0.1 0>,<at 0.2 1>,<at 0.8 1>,<at 1 0>]}}}
       CurveIndex:NiagaraFloat=$Emitter.Scalability.DistanceFraction
       Bias:NiagaraFloat=0
       Scale:NiagaraFloat=0.5
      )
     )
    ))
   &CameraOffset(Camera Offset Amount=2500) {module=CameraOffset001}
   &DynamicMaterialParameters(
    Index 0 Param 1=1
    Index 0 Param 2=2500
    Index 0 Param 3=1500
    Index 0 Param 4=5
    Module.Index0Param0Write=true
    Module.Index0Param1Write=true
    Module.Index0Param2Write=true
    Module.Index0Param3Write=true
   )
  ]
  rendererClasses [&NiagaraSpriteRendererProperties, &NiagaraMeshRendererProperties]
  renderers [
   {rendererClass=&NiagaraSpriteRendererProperties}
   {rendererClass=&NiagaraMeshRendererProperties, rendererIndex=1}
  ]
 }
 Spiral {
  emitterUpdateScript [
   &EmitterState(
    Enable Distance Culling=true
    MaxDistance=25000
    Min Distance Response=::SleepAndLetParticlesFinish
    Module.bApplySpawnCountScaleByDistance=true
    Module.bUseMaxDistance=true
    Module.bUseMinDistance=true
    Scalability Mode=::NewEnumerator1
    Scale Spawn Count=true
    Spawn Count Scale By Distance=DataInterface() {propertyValues {Curve {keys [<at 0 1>,<at 0.1 1>,<at 1 1>]}}}
   )
   &SpawnRate(SpawnRate=&Multiply_Float(A=2, B=$User.SpawnRatMult), Use Spawn Probability=true)
  ]
  particleSpawnScript [
   &InitializeParticle(
    Color=<rgba 0.0578 0.0595 0.0612 1>
    Lifetime Max=5
    Lifetime Min=3
    Lifetime Mode=::NewEnumerator1
    Mesh Scale=<15 15 25>
    Mesh Scale Mode=::NewEnumerator4
    Sprite Size Mode=::NewEnumerator1
    Uniform Sprite Size Max=1
    Uniform Sprite Size Min=1
    UsePositionOffset=false
    Write Material Random=true
   ) {module=InitializeParticle001}
   &InitialMeshOrientation(
    Mesh Orientation Mode=::NewEnumerator2
    Rotation=&RandomRangeVector(Minimum=<0 0 0>, Maximum=<0 0 0>)
    Use Rotation Vector=true
   )
   &MeshRotationForce(Rotation=&RandomRangeVector(Minimum=<0 0 -0.05>, Maximum=<0 0 -0.25>))
   &ApplyInitialForces()
   &CameraOffset(Camera Offset Amount=25)
   &ShapeLocation(Handle Radius=100, Large Radius=250, MaskSpawnGroup=false, Override Seed=false, Shape Primitive=::NewEnumerator3, UseOwnerScale=true) {enabled=false}
   /SetVariables(Particles.InitialPos:NiagaraPosition=$Particles.Position) {isSetParameters=true, module=SetVariables_#2}
   &AddVelocity(Velocity=<0 0 25>)
   &SystemLocation(MaskSpawnGroup=false, Offset=<0 0 -2500>)
   &ShapeLocation(
    MaskSpawnGroup=false
    Override Seed=false
    Ring / Disc Distribution Mode:ENiagaraSphereDistributionMode=::NewEnumerator1
    Ring Radius=0
    Shape Primitive=::NewEnumerator4
    UseOwnerScale=true
   ) {module=ShapeLocation001}
  ]
  particleUpdateScript [
   &ParticleState()
   &SolveForcesAndVelocity(
    Acceleration Limit=25
    Clamp Velocity=true
    Limit Acceleration=true
    Speed Limit:NiagaraFloat=&Multiply_Float(
     A=150
     B=&FloatFromCurve(FloatCurve=DataInterface() {propertyValues {Curve {keys [
          <at 0 0.5>
          {arriveTangent=1.255, interpMode=RCIM_Cubic, leaveTangent=1.255, tangentMode=RCTM_User, time=0.5, value=1}
          {arriveTangent=2.5018, interpMode=RCIM_Cubic, leaveTangent=2.5018, tangentMode=RCTM_User, time=1, value=3}
         ]}}})
    )
   )
   &CalculateAccurateVelocity()
   &ScaleColor(
    Scale Alpha=&FloatFromCurve(
     FloatCurve=DataInterface() {propertyValues {Curve {
        defaultValue=1
        keys [
         {interpMode=RCIM_Cubic, tangentMode=RCTM_User, time=0, value=0}
         {interpMode=RCIM_Cubic, tangentMode=RCTM_User, time=0.5, value=1}
         {tangentMode=RCTM_User, time=1, value=0}
        ]
       }}}
     Scale Curve=1.5
    )
    Scale RGB=&FN_Select_Vector_from_Time_Of_Day(Morning=<25 1 25>, Day=<25 1 25>, Evening=<25 1 25>, Night=<25 1 25>, Evening Threshold=21)
    ScaleA=true
    ScaleRGB=true
   )
   &ScaleMeshSize(Scale Factor=&Multiply_VectorByFloat(
     Vector=&Multiply_VectorByFloat(Vector=<1 1 1>, Float=$User.ScaleMult)
     Float=&FloatFromCurve(FloatCurve=DataInterface() {propertyValues {Curve {
         defaultValue=1
         keys [
          {interpMode=RCIM_Cubic, tangentMode=RCTM_User, time=0, value=0.9}
          {arriveTangent=-0.4076, interpMode=RCIM_Cubic, leaveTangent=1.1984, tangentMode=RCTM_Break, time=1, value=1.1}
         ]
        }}})
    ))
   &CameraOffset(Camera Offset Amount=250) {module=CameraOffset001}
   &DynamicMaterialParameters(
    Index 0 Param 3=&RandomRangeFloat(Minimum=0.5, Maximum=2)
    Index 0 Param 4=&RandomRangeFloat(Minimum=0.1, Maximum=1)
    Module.Index0Float0Write=true
    Module.Index0Float1Write=true
    Module.Index0Float2Write=true
    Module.Index0Float3Write=true
   )
  ]
  rendererClasses [&NiagaraMeshRendererProperties]
  renderers [{rendererClass=&NiagaraMeshRendererProperties, rendererIndex=0}]
 }
}
summary {
 systemName=NS_Spire_Floating_Rocks
 emitters [
  {emitterName=Rocks, simTarget=GPUComputeSim, rendererClasses [&NiagaraMeshRendererProperties]}
  {
   bEnabled=false
   emitterName=Rocks_OnSpire
   simTarget=GPUComputeSim
   rendererClasses [&NiagaraSpriteRendererProperties, &NiagaraMeshRendererProperties]
  }
  {emitterName=Spiral, rendererClasses [&NiagaraMeshRendererProperties]}
 ]
 userVariables [
  User.ScaleMult {type=NiagaraFloat, defaultValue=0.5, description=""}
  User.SpawnRatMult {type=NiagaraFloat, defaultValue=1, description=""}
 ]
}
